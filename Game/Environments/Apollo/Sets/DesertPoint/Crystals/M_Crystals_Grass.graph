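# Masked material for DesertPoint crystal grass in Apollo, usable on instanced static meshes and Niagara mesh particles.
# Base color combines a tint, inverted fresnel and shard mask; emissive adds time-of-day color.
# UV1's vertical gradient blends normals and shading into the landscape runtime virtual texture, behind an RVT switch, with normals converted to world space.
# Forward and mobile paths use a simpler tint and gradient-faded glow without RVT.
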Material /Game/Environments/Apollo/Sets/DesertPoint/Crystals/M_Crystals_Grass
graph {
 properties {
  bTangentSpaceNormal=false
  bUseMaterialAttributes=true
  bUsedWithInstancedStaticMeshes=true
  bUsedWithNiagaraMeshParticles=true
  blendMode=BLEND_Masked
  ditherOpacityMask=true
  pixelDepthOffsetMode=PDOM_Legacy
  usedShadingModels=Default Lit
 }
 nodes [
  n1 = RuntimeVirtualTextureSampleParameter() {materialType=BaseColor_Normal_Specular_YCoCg, parameterName=RVT, virtualTexture=/Game/Athena/Apollo/Maps/Landscape/VT/RVT_Apollo_Landscape}
  n2 = VectorParameter() {parameterName=Color, defaultValue=<rgba 0.0717 0.0649 0.2188 0>}
  n3 = Multiply(A=OneMinus(None=Fresnel() {baseReflectFraction=0.3, exponent=3}), B=@n2.RGB)
  n4 = TextureSampleParameter2D() {
   group=Textures
   parameterName=Mask
   samplerType=SAMPLERTYPE_Masks
   texture=/Game/Environments/Apollo/Sets/DesertPoint/Crystals/Textures/Shards_AB_M
  }
  n5 = VectorParameter() {parameterName=Ecolor2, defaultValue=<rgba 0.6509 0.2091 0.9505 0>}
  n6 = Constant3Vector() {constant=<rgba 0.3545 0.1603 0.9505 0>}
  n7 = Multiply(A=@n4.G, B=ShadingPathSwitch(Default=@n5.RGB, Deferred=@n5.RGB, Forward=@n6, Mobile=@n6))
  n8 = TextureSampleParameter2D() {
   group=Textures
   parameterName=Normal
   samplerType=SAMPLERTYPE_Normal
   texture=/Game/Environments/Apollo/Sets/DesertPoint/Crystals/Textures/Shards_AB_N
  }
  n9 = GetMaterialAttributes(None=MakeMaterialAttributes!(
    BaseColor=Multiply(A=@n3, B=@n4.G)
    Metallic=ScalarParameter() {defaultValue=0.5238, parameterName=Met}
    Specular=ScalarParameter() {defaultValue=0.1, parameterName=S}
    Roughness=ScalarParameter() {defaultValue=0.15, parameterName=R}
    Anisotropy
    EmissiveColor=@n7
    Opacity=PerInstanceFadeAmount()
    OpacityMask
    Normal=@n8.RGB
    Tangent
    WorldPositionOffset
    SubsurfaceColor
    ClearCoat
    ClearCoatRoughness
    AmbientOcclusion
    Refraction
    CustomizedUVs_0
    CustomizedUVs_1
    CustomizedUVs_2
    CustomizedUVs_3
    CustomizedUVs_4
    CustomizedUVs_5
    CustomizedUVs_6
    CustomizedUVs_7
    PixelDepthOffset
    ShadingModel
    Displacement
   )) {attributeGetTypes [#1]}
  n10 = VectorParameter() {group=TOD, parameterName=EmissiveEvening, sortPriority=3, defaultValue=<rgba 0.5208 0.285 0.1768 1>}
  n11 = VectorParameter() {group=TOD, parameterName=EmissiveNight, sortPriority=4, defaultValue=<rgba 0.2726 0.3026 0.3333 1>}
  n12 = MaterialFunctionCall(
   Color_Morning=VectorParameter() {group=TOD, parameterName=EmissiveMorning, sortPriority=1, defaultValue=<rgba 1 0.4297 0.5294 1>}.RGB
   Color_Day=VectorParameter() {group=TOD, parameterName=EmissiveDay, sortPriority=2, defaultValue=<rgba 1 1 1 1>}.RGB
   Color_Evening=@n10.RGB
   Color_Night=@n11.RGB
  ) {materialFunction=&Color_Time_of_Day}
  n13 = MaterialFunctionCall!(Material=SetMaterialAttributes(MaterialAttributes=@n9.MaterialAttributes, Emissive Color=Multiply(A=@n9.Emissive Color, B=@n12.Result)) {attributeSetTypes [#1]}) {materialFunction=/Game/Athena/Apollo/Maps/Landscape/VT/MaterialFunctions/TransformTangentToWorld}
  n14 = GetMaterialAttributes(None=@n13.Result) {attributeGetTypes [#1]}
  n15 = OneMinus(None=ComponentMask(None=TextureCoordinate() {coordinateIndex=1}) {g=true})
  n16 = ScalarParameter() {defaultValue=35, parameterName=Blend RVT N}
  n17 = BlendMaterialAttributes(
   A=SetMaterialAttributes!(MaterialAttributes, Normal=@n1.Normal) {attributeSetTypes [#2]}
   B=SetMaterialAttributes!(MaterialAttributes=@n14.MaterialAttributes, Emissive Color=Multiply(A=@n14.Emissive Color, B=@n1.BaseColor)) {attributeSetTypes [#1]}
   Alpha=Saturate(None=Divide(A=@n15, B=@n16))
  ) {vertexAttributeBlendType=UseA}
  n18 = Divide(A=@n15, B=ScalarParameter() {defaultValue=55, parameterName=Blend RVT})
  n19 = MaterialFunctionCall(
   NoRVT=@n13.Result
   RVT=BlendMaterialAttributes(
    A=SetMaterialAttributes!(MaterialAttributes=@n17, Base Color=@n1.BaseColor, Specular=@n1.BaseColor, Roughness=@n1.BaseColor) {attributeSetTypes [#3, #4, #5]}
    B=@n17
    Alpha=Saturate(None=@n18)
   ) {vertexAttributeBlendType=UseA}
  ) {materialFunction=&FN_RVT_Switch_Material}
  n20 = ComponentMask(None=TextureCoordinate() {coordinateIndex=1}) {g=true}
  n21 = OneMinus(None=@n20)
  n22 = MaterialFunctionCall(Material=SetMaterialAttributes!(MaterialAttributes, Base Color=Multiply(A=@n2.RGB, B=@n4.G), Emissive Color=Multiply(A=@n7, B=@n21), Roughness=@n20) {attributeSetTypes [#3, #1, #5]}) {materialFunction=&TransformTangentToWorld}
  ShadingPathSwitch(Default=@n19.Result, Deferred=@n19.Result, Forward=@n22.Result, Mobile=@n22.Result)
 ]
}
